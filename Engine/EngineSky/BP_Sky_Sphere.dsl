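; Sky sphere Blueprint that drives a dynamic panning-cloud sky material with sun, color, cloud and star parameters.
; The construction script creates the material instance and calls RefreshMaterial.
; RefreshMaterial derives sun direction and color from the assigned directional light, or from manual sun height without one, then sets sky colors from sun-height curves or overrides.
; UpdateSunDirection reapplies only sun and curve-driven parameters, assuming a light is assigned.
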(fn RefreshMaterial
  (bind _sky_material (Variables|Default|GetSky%20material))
  (bind _sky_material_2 (Variables|Default|GetSky%20material))
  (bind _sky_material_3 (Variables|Default|GetSky%20material))
  (bind _sun_height (Variables|Overridesettings|GetSun%20height))
  (bind _sky_material_4 (Variables|Default|GetSky%20material))
  (bind _sky_material_5 (Variables|Default|GetSky%20material))
  (Utilities|IsValid (Variables|Default|GetDirectional%20light%20actor)
    (:Is%20Valid
      (Rendering|Material|SetVectorParameterValue _sky_material_3 "Light direction" (Math|Conversions|ToLinearColor(Vector) (Math|Rotator|GetRotationXVector (Transformation|GetActorRotation (Variables|Default|GetDirectional%20light%20actor)))) :!class /Script/Engine.MaterialInstanceDynamic)
      (Rendering|Material|SetVectorParameterValue _sky_material_3 "Sun color" (Math|Conversions|ToLinearColor(Color) (Variables|/Script/Engine.LightComponentBase|GetLightColor (Variables|/Script/Engine.Light|GetLightComponent (Variables|Default|GetDirectional%20light%20actor)))) :!class /Script/Engine.MaterialInstanceDynamic)
      (Variables|Overridesettings|SetSun%20height (Math|Float|MapRangeUnclamped (.pitch (Transformation|GetActorRotation (Variables|Default|GetDirectional%20light%20actor))) 0.0 -90 0.0 1)))
    (:Is%20Not%20Valid
      (Rendering|Material|SetVectorParameterValue _sky_material_2 "Light direction" (Math|Conversions|ToLinearColor(Vector) (Math|Rotator|GetRotationXVector (Math|Rotator|MakeRotator 0.0 (Math|Float|MapRangeUnclamped (Variables|Overridesettings|GetSun%20height) -1 1 90 -90) (.yaw (Transformation|GetActorRotation))))) :!class /Script/Engine.MaterialInstanceDynamic)
      (Rendering|Material|SetVectorParameterValue _sky_material_2 "Sun color" (Math|Color|Lerp(LinearColor) "(R=1.000000,G=0.221000,B=0.040000,A=1.000000)" "(R=0.954000,G=0.901000,B=0.744120,A=1.000000)" (Math|Float|Clamp(Float) (+ (Variables|Overridesettings|GetSun%20height) 0.2))) :!class /Script/Engine.MaterialInstanceDynamic))) ; Has a directional light been assigned?
  (if (Variables|Default|GetColors%20determined%20by%20sun%20position)
    (Rendering|Material|SetVectorParameterValue _sky_material_4 "Horizon color" (Math|Curves|GetClampedLinearColorValue (Variables|Default|GetHorizon%20color%20curve) (Variables|Overridesettings|GetSun%20height)) :!class /Script/Engine.MaterialInstanceDynamic)
    (Rendering|Material|SetVectorParameterValue _sky_material_4 "Zenith color" (Math|Curves|GetClampedLinearColorValue (Variables|Default|GetZenith%20color%20curve) (Variables|Overridesettings|GetSun%20height)) :!class /Script/Engine.MaterialInstanceDynamic)
    (Rendering|Material|SetVectorParameterValue _sky_material_4 "Cloud color" (Math|Curves|GetClampedLinearColorValue (Variables|Default|GetCloud%20color%20curve) _sun_height) :!class /Script/Engine.MaterialInstanceDynamic)
    (Rendering|Material|SetScalarParameterValue _sky_material_4 "Horizon falloff" (Math|Float|Lerp 3 7 (Math|Float|Absolute(Float) _sun_height)) :!class /Script/Engine.MaterialInstanceDynamic)
    (else
      (Rendering|Material|SetVectorParameterValue _sky_material "Horizon color" (Variables|Overridesettings|GetHorizon%20color) :!class /Script/Engine.MaterialInstanceDynamic)
      (Rendering|Material|SetVectorParameterValue _sky_material "Zenith color" (Variables|Overridesettings|GetZenith%20Color) :!class /Script/Engine.MaterialInstanceDynamic)
      (Rendering|Material|SetVectorParameterValue _sky_material "Overall color" (Variables|Overridesettings|GetOverall%20Color) :!class /Script/Engine.MaterialInstanceDynamic)
      (Rendering|Material|SetVectorParameterValue _sky_material "Cloud color" (Variables|Overridesettings|GetCloud%20color) :!class /Script/Engine.MaterialInstanceDynamic)
      (Rendering|Material|SetScalarParameterValue _sky_material "Horizon falloff" (Variables|Overridesettings|GetHorizon%20Falloff) :!class /Script/Engine.MaterialInstanceDynamic)))
  (Rendering|Material|SetScalarParameterValue _sky_material_5 "Cloud speed" (Variables|Default|GetCloud%20speed) :!class /Script/Engine.MaterialInstanceDynamic)
  (Rendering|Material|SetScalarParameterValue _sky_material_5 "Sun brightness" (Variables|Default|GetSun%20brightness) :!class /Script/Engine.MaterialInstanceDynamic)
  (Rendering|Material|SetScalarParameterValue _sky_material_5 "Sun height" (Math|Float|SelectFloat (Math|Float|Absolute(Float) (Variables|Overridesettings|GetSun%20height)) 0.0 (< (Variables|Overridesettings|GetSun%20height) 0.0)) :!class /Script/Engine.MaterialInstanceDynamic)
  (Rendering|Material|SetScalarParameterValue _sky_material_5 "Cloud opacity" (Variables|Default|GetCloud%20opacity) :!class /Script/Engine.MaterialInstanceDynamic)
  (Rendering|Material|SetScalarParameterValue _sky_material_5 "Stars brightness" (Variables|Default|GetStars%20brightness) :!class /Script/Engine.MaterialInstanceDynamic))

(fn UpdateSunDirection
  (bind _sky_material (Variables|Default|GetSky%20material))
  (bind _sun_height (Variables|Overridesettings|GetSun%20height))
  (bind _sky_material_2 (Variables|Default|GetSky%20material))
  (Rendering|Material|SetVectorParameterValue _sky_material "Light direction" (Math|Conversions|ToLinearColor(Vector) (Math|Rotator|GetRotationXVector (Transformation|GetActorRotation (Variables|Default|GetDirectional%20light%20actor)))) :!class /Script/Engine.MaterialInstanceDynamic)
  (Rendering|Material|SetVectorParameterValue _sky_material "Sun color" (Math|Conversions|ToLinearColor(Color) (Variables|/Script/Engine.LightComponentBase|GetLightColor (Variables|/Script/Engine.Light|GetLightComponent (Variables|Default|GetDirectional%20light%20actor)))) :!class /Script/Engine.MaterialInstanceDynamic)
  (Variables|Overridesettings|SetSun%20height (Math|Float|MapRangeUnclamped (.pitch (Transformation|GetActorRotation (Variables|Default|GetDirectional%20light%20actor))) 0.0 -90 0.0 1))
  (Rendering|Material|SetVectorParameterValue _sky_material_2 "Horizon color" (Math|Curves|GetClampedLinearColorValue (Variables|Default|GetHorizon%20color%20curve) (Variables|Overridesettings|GetSun%20height)) :!class /Script/Engine.MaterialInstanceDynamic)
  (Rendering|Material|SetVectorParameterValue _sky_material_2 "Zenith color" (Math|Curves|GetClampedLinearColorValue (Variables|Default|GetZenith%20color%20curve) (Variables|Overridesettings|GetSun%20height)) :!class /Script/Engine.MaterialInstanceDynamic)
  (Rendering|Material|SetVectorParameterValue _sky_material_2 "Cloud color" (Math|Curves|GetClampedLinearColorValue (Variables|Default|GetCloud%20color%20curve) _sun_height) :!class /Script/Engine.MaterialInstanceDynamic)
  (Rendering|Material|SetScalarParameterValue _sky_material_2 "Horizon falloff" (Math|Float|Lerp 3 8 (Math|Float|Absolute(Float) _sun_height)) :!class /Script/Engine.MaterialInstanceDynamic)
  (Rendering|Material|SetScalarParameterValue _sky_material_2 "Sun height" (Math|Float|SelectFloat (Math|Float|Absolute(Float) (Variables|Overridesettings|GetSun%20height)) 0.0 (< (Variables|Overridesettings|GetSun%20height) 0.0)) :!class /Script/Engine.MaterialInstanceDynamic))

(fn ConstructionScript
  (bind _returnvalue (Rendering|Material|CreateDynamicMaterialInstance (Variables|BP_Skydome|GetSkySphereMesh) 0 "/Engine/EngineSky/M_Sky_Panning_Clouds2.M_Sky_Panning_Clouds2" "None" :!class /Script/Engine.PrimitiveComponent))
  (Variables|Default|SetSky%20material _returnvalue)
  (if (Variables|Default|GetRefresh%20material)
    _)
  (Variables|Default|SetRefresh%20material false)
  (CallFunction|RefreshMaterial))
