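# Unlit two-sided sky dome material that draws the Sky Atmosphere view and two light disks behind a procedural cloud layer.
# Scrolling noise becomes cloud density, feeding Custom HLSL for per-light phase-weighted scattering and transmittance, plus ambient from distant sky light.
# Clouds receive aerial perspective, transmittance attenuates the sky, and output is capped for small-float range.
# The world-height cloud mask multiplies by zero, leaving clouds invisible as saved.
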Material /Engine/EngineSky/M_AdvancedSkyDome
graph {
 properties {
  bIsSky=true
  bUsedWithInstancedStaticMeshes=true
  bUsedWithStaticLighting=true
  pixelDepthOffsetMode=PDOM_Legacy
  shadingModel=MSM_Unlit
  twoSided=true
  usedShadingModels=Unlit
 }
 nodes [
  MaterialExpressionSkyAtmosphereAerialPerspective_0 = SkyAtmosphereAerialPerspective!(World Position)
  n1 = Constant()
  n2 = TextureCoordinate()
  n3 = Time()
  MaterialExpressionTextureSample_1 = TextureSample(UVs=Add(A=Multiply(A=@n2) {constB=100}, B=Multiply(A=@n3) {constB=0.0015})) {texture=/Engine/MapTemplates/TilingNoise05}
  MaterialExpressionCustom_0 = Custom!(CloudDensity=@MaterialExpressionTextureSample_1.R) {
   code="return saturate((CloudDensity-0.1)*1.1);"
   description=Cloud Density
   inputs [{
     inputName=CloudDensity
     input {expression=&self:MaterialExpressionTextureSample_1, mask=1, maskR=1, outputIndex=1}
    }]
  }
  MaterialExpressionScalarParameter_3 = ScalarParameter() {defaultValue=1, parameterName=CloudHeight}
  MaterialExpressionSkyAtmosphereLightDirection_2 = SkyAtmosphereLightDirection()
  MaterialExpressionSkyAtmosphereLightIlluminance_2 = SkyAtmosphereLightIlluminance!(World Position)
  MaterialExpressionCameraPositionWS_1 = CameraPositionWS()
  MaterialExpressionScalarParameter_4 = ScalarParameter() {defaultValue=0.7, parameterName=ScatteringCoeff}
  MaterialExpressionScalarParameter_5 = ScalarParameter() {parameterName=AbsorbtionCoeff}
  MaterialExpressionWorldPosition_0 = WorldPosition()
  n4 = Custom!(
   CloudDensity=@MaterialExpressionCustom_0
   CloudHeight=@MaterialExpressionScalarParameter_3
   SunLightDir=@MaterialExpressionSkyAtmosphereLightDirection_2
   SunLightIlluminance=@MaterialExpressionSkyAtmosphereLightIlluminance_2
   CameraPos=@MaterialExpressionCameraPositionWS_1
   ScatteringCoeff=@MaterialExpressionScalarParameter_4
   AbsorptionCoeff=@MaterialExpressionScalarParameter_5
   WorldPos=@MaterialExpressionWorldPosition_0.XYZ
  ) {
   code="float3 sigmaS = ScatteringCoeff * CloudDensity;\r\nfloat3 sigmaA = AbsorptionCoeff * CloudDensity;\r\nfloat3 sigmaE = sigmaS + sigmaA + 1e-6;\r\n\r\nfloat3 TransmittanceRGB = exp(-CloudHeight * sigmaE);\r\n\r\nfloat phaseIn = saturate(dot(SunLightDir, -normalize(CameraPos-WorldPos)));\r\nfloat phaseUni = 1.0;\r\nfloat k = 0.8;\r\nfloat factor = 1.0 - k * phaseIn;\r\nfloat phaseSchlick = (1.0 - k*k) / (factor*factor);\r\n\r\nfloat phase = (0.3*phaseUni + 0.7*phaseSchlick) / (4.0 * PI);\r\n\r\nconst float boost = 1.0f;\r\n\r\nfloat3 S = SunLightIlluminance * phase * sigmaS * boost;\r\nfloat3 ScattLuminance = (S - S * TransmittanceRGB) / sigmaE;\r\n\r\n\r\nfloat3 ScattLumFromTop = ScattLuminance * TransmittanceRGB;\r\n\r\n// TODO We need an ambient color matchign the sky mean illuminance. Otherwise the clouds remain black in area the sun light cannot reach.\r\n\r\nreturn max(0.0, ScattLumFromTop);"
   description=Cloud material
   inputs [
    {inputName=CloudDensity, input {expression=&self:MaterialExpressionCustom_0}}
    {inputName=CloudHeight, input {expression=&self:MaterialExpressionScalarParameter_3}}
    {inputName=SunLightDir, input {expression=&self:MaterialExpressionSkyAtmosphereLightDirection_2}}
    {inputName=SunLightIlluminance, input {expression=&self:MaterialExpressionSkyAtmosphereLightIlluminance_2}}
    {inputName=CameraPos, input {expression=&self:MaterialExpressionCameraPositionWS_1}}
    {inputName=ScatteringCoeff, input {expression=&self:MaterialExpressionScalarParameter_4}}
    {inputName=AbsorptionCoeff, input {expression=&self:MaterialExpressionScalarParameter_5}}
    {inputName=WorldPos, input {expression=&self:MaterialExpressionWorldPosition_0}}
   ]
  }
  n5 = WorldPosition()
  n6 = Saturate(None=Multiply(A=MaterialFunctionCall(Color=@n5.XYZ) {materialFunction=&SplitComponents}.B) {constB=0})
  MaterialExpressionSkyAtmosphereLightDirection_3 = SkyAtmosphereLightDirection() {lightIndex=1}
  MaterialExpressionSkyAtmosphereLightIlluminance_3 = SkyAtmosphereLightIlluminance() {lightIndex=1}
  n7 = Custom(
   CloudDensity=@MaterialExpressionCustom_0
   CloudHeight=@MaterialExpressionScalarParameter_3
   SunLightDir=@MaterialExpressionSkyAtmosphereLightDirection_3
   SunLightIlluminance=@MaterialExpressionSkyAtmosphereLightIlluminance_3
   CameraPos=@MaterialExpressionCameraPositionWS_1
   ScatteringCoeff=@MaterialExpressionScalarParameter_4
   AbsorptionCoeff=@MaterialExpressionScalarParameter_5
   WorldPos=@MaterialExpressionWorldPosition_0.XYZ
  ) {
   code="float3 sigmaS = ScatteringCoeff * CloudDensity;\r\nfloat3 sigmaA = AbsorptionCoeff * CloudDensity;\r\nfloat3 sigmaE = sigmaS + sigmaA + 1e-6;\r\n\r\nfloat3 TransmittanceRGB = exp(-CloudHeight * sigmaE);\r\n\r\nfloat phaseIn = saturate(dot(SunLightDir, -normalize(CameraPos-WorldPos)));\r\nfloat phaseUni = 1.0;\r\nfloat k = 0.8;\r\nfloat factor = 1.0 - k * phaseIn;\r\nfloat phaseSchlick = (1.0 - k*k) / (factor*factor);\r\n\r\nfloat phase = (0.3*phaseUni + 0.7*phaseSchlick) / (4.0 * PI);\r\n\r\nconst float boost = 1.0f;\r\n\r\nfloat3 S = SunLightIlluminance * phase * sigmaS * boost;\r\nfloat3 ScattLuminance = (S - S * TransmittanceRGB) / sigmaE;\r\n\r\n\r\nfloat3 ScattLumFromTop = ScattLuminance * TransmittanceRGB;\r\n\r\n// TODO We need an ambient color matchign the sky mean illuminance. Otherwise the clouds remain black in area the sun light cannot reach.\r\n\r\nreturn max(0.0, ScattLumFromTop);"
   description=Cloud material
   inputs [
    {inputName=CloudDensity, input {expression=&self:MaterialExpressionCustom_0}}
    {inputName=CloudHeight, input {expression=&self:MaterialExpressionScalarParameter_3}}
    {inputName=SunLightDir, input {expression=&self:MaterialExpressionSkyAtmosphereLightDirection_3}}
    {inputName=SunLightIlluminance, input {expression=&self:MaterialExpressionSkyAtmosphereLightIlluminance_3}}
    {inputName=CameraPos, input {expression=&self:MaterialExpressionCameraPositionWS_1}}
    {inputName=ScatteringCoeff, input {expression=&self:MaterialExpressionScalarParameter_4}}
    {inputName=AbsorptionCoeff, input {expression=&self:MaterialExpressionScalarParameter_5}}
    {inputName=WorldPos, input {expression=&self:MaterialExpressionWorldPosition_0}}
   ]
  }
  n8 = SkyAtmosphereDistantLightScatteredLuminance()
  n9 = ScalarParameter() {parameterName=Fraction}
  n10 = Custom!(
   CloudDensity=@MaterialExpressionCustom_0
   CloudHeight=@MaterialExpressionScalarParameter_3
   ScatteringCoeff=@MaterialExpressionScalarParameter_4
   AbsorptionCoeff=@MaterialExpressionScalarParameter_5
  ) {
   code="float3 sigmaS = ScatteringCoeff * CloudDensity;\r\nfloat3 sigmaA = AbsorptionCoeff * CloudDensity;\r\nfloat3 sigmaE = sigmaS + sigmaA;\r\n\r\nfloat3 TransmittanceRGB = exp(-CloudHeight * sigmaE);\r\n\r\nfloat TransmittanceA = dot(TransmittanceRGB, float3(1.0/3.0, 1.0/3.0, 1.0/3.0));\r\n\r\nreturn saturate(TransmittanceA);"
   description=Cloud transmittance
   outputType=CMOT_Float1
   inputs [
    {inputName=CloudDensity, input {expression=&self:MaterialExpressionCustom_0}}
    {inputName=CloudHeight, input {expression=&self:MaterialExpressionScalarParameter_3}}
    {inputName=ScatteringCoeff, input {expression=&self:MaterialExpressionScalarParameter_4}}
    {inputName=AbsorptionCoeff, input {expression=&self:MaterialExpressionScalarParameter_5}}
   ]
  }
  MaterialExpressionLinearInterpolate_2 = LinearInterpolate(A=Constant() {r=1}, B=@n10, Alpha=@n6)
  MaterialExpressionAdd_7 = Add(
   A=Add(A=LinearInterpolate(A=@n1, B=@n4, Alpha=@n6), B=LinearInterpolate(A=@n1, B=@n7, Alpha=@n6))
   B=Multiply(A=Desaturation(None=@n8, Fraction=@n9), B=OneMinus(None=@MaterialExpressionLinearInterpolate_2)) {constA=1, desc=Ambient x coverage}
  ) {constB=0}
  n11 = Custom!(
   AerialPerspective=@MaterialExpressionSkyAtmosphereAerialPerspective_0
   CloudLuminance=@MaterialExpressionAdd_7
   CloudTransmittance=@MaterialExpressionLinearInterpolate_2
  ) {
   code="float3 Result;\r\n\r\nfloat CloudOpacity = 1.0 - CloudTransmittance;\r\n\r\nResult = \r\nCloudOpacity * AerialPerspective.rgb\r\n+\r\nCloudLuminance.rgb*AerialPerspective.a;\r\n\r\nreturn Result;"
   description=ApplyAerialPErspectiveOnClouds
   inputs [
    {inputName=AerialPerspective, input {expression=&self:MaterialExpressionSkyAtmosphereAerialPerspective_0}}
    {inputName=CloudLuminance, input {expression=&self:MaterialExpressionAdd_7}}
    {inputName=CloudTransmittance, input {expression=&self:MaterialExpressionLinearInterpolate_2}}
   ]
  }
  n12 = Add(
   A=SkyAtmosphereViewLuminance!(WorldDirection)
   B=Add(A=SkyAtmosphereLightDiskLuminance!(DiskAngularDiameterOverride), B=SkyAtmosphereLightDiskLuminance() {lightIndex=1})
  )
  Min(A=Add(A=@n11, B=Multiply(A=@MaterialExpressionLinearInterpolate_2, B=@n12))) {
   constB=64000
   desc="Do not go over small float\r\nrepresentation limit.\r\nThis is because the sun disk \r\nluminance can be high.\r\nIdeally the high bound should also\r\ndepend on exposure but we do not have\r\nthat information."
  }
 ]
}
