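# Simple sky dome material that shows the Sky Atmosphere's sky and light disk.
# It is unlit and flagged as a sky material.
# A single Add sums the Sky Atmosphere view luminance and light disk luminance nodes, with no textures or parameters.
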Material /Engine/EngineSky/M_SimpleSkyDome
graph {
 properties {bIsSky=true, pixelDepthOffsetMode=PDOM_Legacy, shadingModel=MSM_Unlit, usedShadingModels=Unlit}
 nodes [Add(A=SkyAtmosphereViewLuminance!(WorldDirection), B=SkyAtmosphereLightDiskLuminance!(DiskAngularDiameterOverride))]
}
