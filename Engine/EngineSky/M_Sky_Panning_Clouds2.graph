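# Unlit sky dome material drawing a zenith-to-horizon gradient, stars, a sun disc and animated clouds.
# View direction drives the gradient and sphere-masks the sun against a light direction parameter.
# Two cloud textures pan at different rates, blend by view elevation, fade below the mesh center and get noise-varied contrast.
# Clouds gain a sun-facing glow, then composite over the sky with an overall tint.
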Material /Engine/EngineSky/M_Sky_Panning_Clouds2
graph {
 properties {bFullyRough=true, bUsedWithStaticLighting=true, pixelDepthOffsetMode=PDOM_Legacy, shadingModel=MSM_Unlit, usedShadingModels=Unlit}
 nodes [
  n1 = VectorParameter() {group=Sky, parameterName=Zenith Color, defaultValue=<rgba 0.0852 0.1537 0.35 1>}
  n2 = TextureCoordinate() {coordinateIndex=1, uTiling=12, vTiling=12}
  n3 = Multiply(
   A=Multiply(A=TextureSample(UVs=@n2) {texture=&/T_Sky_Stars}.RGB, B=ScalarParameter() {defaultValue=0.1, parameterName=Stars brightness}) {constB=0.02}
   B=ScalarParameter() {defaultValue=1, parameterName=Sun height}
  )
  n4 = Clamp(None=DotProduct(A=CameraVectorWS(), B=Constant3Vector() {constant=<rgba 0 0 -1 0>}))
  n5 = Power(Base=OneMinus(None=@n4), Exp=ScalarParameter() {defaultValue=3, group=Sky, parameterName=Horizon Falloff})
  n6 = DotProduct(A=CameraVectorWS(), B=Normalize(VectorInput=VectorParameter() {group=Sun, parameterName=Light direction, defaultValue=<rgba -2 0 -1 1>}.RGB))
  n7 = Constant() {r=1}
  n8 = ScalarParameter() {defaultValue=0.0003, group=Sun, parameterName=Sun Radius}
  n9 = VectorParameter() {group=Sun, parameterName=Sun color, defaultValue=<rgba 1 0.8 0.4 1>}
  n10 = ScalarParameter() {defaultValue=50, parameterName=Sun brightness}
  n11 = Add(
   A=LinearInterpolate(
    A=Add(A=@n1.RGB, B=@n3)
    B=VectorParameter() {group=Sky, parameterName=Horizon color, defaultValue=<rgba 0.9406 1 1 1>}.RGB
    Alpha=Clamp(None=@n5)
   )
   B=Multiply(A=SphereMask(A=@n6, B=@n7, Radius=@n8) {attenuationRadius=0.001, hardnessPercent=0}, B=Multiply(A=@n9.RGB, B=@n10))
  )
  n12 = Time()
  n13 = ScalarParameter() {defaultValue=0.1, group=Animation, parameterName=Cloud speed}
  n14 = TextureSample(UVs=Panner(Time=Multiply(A=@n12, B=@n13)) {speedX=0.0002}) {texture=&/T_Sky_Blue}
  n15 = TextureSample(UVs=Panner(Time=Multiply(A=@n12, B=@n13)) {speedX=0.001}) {samplerType=SAMPLERTYPE_Grayscale, texture=&/T_Sky_Clouds_M}
  n16 = Divide(A=Subtract(A=WorldPosition().XYZ, B=ObjectPositionWS()), B=Multiply(A=ObjectRadius()) {constB=-0.1}) {constB=2000}
  n17 = OneMinus(None=Clamp(None=ComponentMask(None=@n16) {b=true}))
  n18 = ScalarParameter() {defaultValue=1, parameterName=Cloud opacity}
  n19 = TextureCoordinate() {uTiling=0.5, vTiling=0.5}
  n20 = Power(
   Base=LinearInterpolate(B=LinearInterpolate(A=@n14.R, B=@n15.R, Alpha=@n4), Alpha=Multiply(A=@n17, B=@n18))
   Exp=LinearInterpolate(
    A=ScalarParameter() {defaultValue=1, group=Animation, parameterName=NoisePower1}
    B=ScalarParameter() {defaultValue=4, group=Animation, parameterName=NoisePower2}
    Alpha=TextureSample(UVs=@n19) {samplerType=SAMPLERTYPE_Grayscale, texture=&T_Sky_Clouds_M}.R
   ) {constA=0.1, constB=4}
  )
  n21 = Constant() {r=1}
  n22 = Constant() {r=1.3}
  n23 = Clamp(None=Power(Base=SphereMask(A=@n6, B=@n21, Radius=@n22) {attenuationRadius=0, hardnessPercent=0}) {constExponent=10})
  n24 = Multiply(A=@n20, B=@n20) {constB=0.5}
  n25 = Add(
   A=Multiply(A=VectorParameter() {group=CloudColors, parameterName=Cloud color, defaultValue=<rgba 0.7168 0.7822 0.885 0>}.RGB, B=@n20)
   B=Multiply(A=Multiply(A=@n23, B=@n9.RGB), B=Multiply(A=@n24) {constB=0.4})
  )
  n26 = Clamp(None=@n24)
  Multiply(A=Multiply(
    A=LinearInterpolate(A=@n11, B=@n25, Alpha=@n26)
    B=VectorParameter() {group=Global, parameterName=Overall Color, defaultValue=<rgba 1 1 1 1>}.RGB
   )) {constB=1.5}
 ]
}
